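# Animation Blueprint for the female medium IronLilac Shade cosmetic body, following the attached parent mesh's pose.
# The copied pose feeds the body's Control Rig, which receives emote and parachute states.
# Two always-on additive bone edits then nudge and pitch dyn_cans.
# A final edit enlarges neck_01_mask only at LOD 1, using a flag the event graph refreshes each animation update.
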AnimBlueprint /BRCosmetics/Characters/Player/Female/Medium/Bodies/F_MED_IronLilac_Shade/Meshes/F_MED_IronLilac_Shade_AnimBP
target_skeleton=&/F_MED_IronLilac_Shade_Skeleton
graphs [
 AnimGraph {nodes [
   n1 = Control Rig!(
    Source=Copy Pose From Mesh() {bUseAttachedParent=true}
    Alpha
    isEmote:bool=Get bIsPlayingEmote()
    isParashuting:bool=Get bIsParachuteOpened()
   ) -> (Pose) {
    alphaBoolBlend {blendOption=Linear}
    controlRigAssetReference {blueprintRigClass=&/F_MED_IronLilac_Shade_CtrlRig.F_MED_IronLilac_Shade_CtrlRig_C}
   }
   n2 = "Transform (Modify) Bone - Bone: dyn_cans"(ComponentPose=Local To Component(LocalPose=@n1), Alpha=1, Translation="0,0,-0.5", Rotation="0,0,0", Scale="1,1,1") {
    rotationMode=BMM_Additive
    rotationSpace=BCS_BoneSpace
    translationMode=BMM_Additive
    translationSpace=BCS_BoneSpace
    alphaBoolBlend {blendOption=Linear}
    boneToModify {boneName=dyn_cans}
    translation {z=-0.5}
   }
   n3 = "Transform (Modify) Bone - Bone: dyn_cans"!(ComponentPose=@n2, bAlphaBoolEnabled:bool=True, Translation="0,0,-0.5", Rotation="-7,0,0", Scale="1,1,1") -> (Pose) {
    alphaInputType=Bool
    rotationMode=BMM_Additive
    rotationSpace=BCS_BoneSpace
    translationMode=BMM_Additive
    translationSpace=BCS_BoneSpace
    alphaBoolBlend {blendOption=Linear}
    boneToModify {boneName=dyn_cans}
    rotation {pitch=-7}
    translation {z=-0.5}
   }
   n4 = "Transform (Modify) Bone - Bone: neck_01_mask"!(ComponentPose=@n3, bAlphaBoolEnabled:bool=Get LOD1() -> (LOD1:bool), Translation="0,0,0", Rotation="0,0,0", Scale="1.3,1.3,1.3") -> (Pose) {
    alphaInputType=Bool
    rotationMode=BMM_Additive
    rotationSpace=BCS_BoneSpace
    scaleMode=BMM_Replace
    scaleSpace=BCS_BoneSpace
    translationMode=BMM_Additive
    translationSpace=BCS_BoneSpace
    alphaBoolBlend {blendOption=Linear}
    boneToModify {boneName=neck_01_mask}
    scale=<1.3 1.3 1.3>
   }
   Output Pose(Result=Component To Local(ComponentPose=@n4)) {layerGroup=None}
  ]}
 EventGraph {nodes [
   Event BlueprintUpdateAnimation()
   > Set LOD1(LOD1:bool="Equal (Integer)"(A=Get CurrentLOD(), B=1)) -> (Output_Get:bool)
  ]}
]
variables [LOD1 {type=boolean}]
